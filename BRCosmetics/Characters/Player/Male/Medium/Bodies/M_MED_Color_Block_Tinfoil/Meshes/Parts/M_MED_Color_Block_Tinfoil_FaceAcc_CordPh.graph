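# Physics asset for a dangling cord on the face accessory of the male medium Color Block Tinfoil cosmetic character.
# A capsule on the cord bone hangs from a small sphere on the cord master bone, with limited swing and a tight twist limit.
# The master links to a chain of torso capsules on spine_02, spine_01 and pelvis through constraints with free swing and twist.
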PhysicsAsset /BRCosmetics/Characters/Player/Male/Medium/Bodies/M_MED_Color_Block_Tinfoil/Meshes/Parts/M_MED_Color_Block_Tinfoil_FaceAcc_CordPh
bodies {
 dyn_cord_01 [{shapeType=Capsule, length=18.755, radius=1.38, shapeName=dyn_cord_01_capsule, center=<14.7404 0 0>, rotation=<rot 90 0 0>}]
 dyn_cord_MSTR [{shapeType=Sphere, radius=0.505, shapeName=dyn_cord_MSTR_sphere}]
 pelvis [{shapeType=Capsule, length=16.125, radius=6.375, shapeName=pelvis_capsule}]
 spine_01 [{shapeType=Capsule, length=15.875, radius=6.625, shapeName=spine_01_capsule}]
 spine_02 [{shapeType=Capsule, length=16, radius=6.5, shapeName=spine_02_capsule}]
}
constraints [
 {bone1Name=dyn_cord_01, bone2Name=dyn_cord_MSTR, swing1LimitDegrees=45, swing2LimitDegrees=60, twistLimitDegrees=5}
 {
  bone1Name=dyn_cord_MSTR
  bone2Name=spine_02
  swing1LimitDegrees=45
  swing1Motion=Free
  swing2LimitDegrees=45
  swing2Motion=Free
  twistLimitDegrees=45
  twistMotion=Free
 }
 {
  bone1Name=spine_01
  bone2Name=pelvis
  swing1LimitDegrees=45
  swing1Motion=Free
  swing2LimitDegrees=45
  swing2Motion=Free
  twistLimitDegrees=45
  twistMotion=Free
 }
 {
  bone1Name=spine_02
  bone2Name=spine_01
  swing1LimitDegrees=45
  swing1Motion=Free
  swing2LimitDegrees=45
  swing2Motion=Free
  twistLimitDegrees=45
  twistMotion=Free
 }
]
