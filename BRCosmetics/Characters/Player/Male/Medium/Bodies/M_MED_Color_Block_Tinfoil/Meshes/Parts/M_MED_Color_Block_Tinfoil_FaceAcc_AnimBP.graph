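# Post-copy AnimBP for the Color Block Tinfoil face accessory, posing the hat, headphones, walkman and cord on a player character.
# It copies the parent pose, layering walkman and cord bones from the body mesh found when customization loads.
# Neck and head rotations relative to spine_05 drive corrective hat and headphone offsets, alongside a headphone constraint.
# A RigidBody node simulates with a cord physics asset; when rigid bodies are disabled, a fixed cord-bone transform takes over.
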AnimBlueprint /BRCosmetics/Characters/Player/Male/Medium/Bodies/M_MED_Color_Block_Tinfoil/Meshes/Parts/M_MED_Color_Block_Tinfoil_FaceAcc_AnimBP
target_skeleton=&/M_MED_Color_Block_Tinfoil_FaceAcc_Skeleton
graphs [
 AnimGraph {nodes [
   "Transform (Modify) Bone - Bone: dyn_headphones_MSTR"(Alpha=Get NeckRotRight() -> (NeckRotRight:double), Translation="0,0,0", Rotation="0.000,0,-6", Scale="1,1,1") {
    rotationMode=BMM_Additive
    rotationSpace=BCS_BoneSpace
    translationMode=BMM_Additive
    translationSpace=BCS_BoneSpace
    alphaBoolBlend {blendOption=Linear}
    boneToModify {boneName=dyn_headphones_MSTR}
    rotation {roll=-6}
   }
   n1 = "Transform (Modify) Bone - Bone: dyn_hat"(
    ComponentPose=Local To Component(LocalPose=Use cached pose 'Start Pose'())
    Alpha=Get HeadBackRot() -> (HeadBackRot:double)
    Translation="-1,0.25,0"
    Rotation="0.000,-3,0.000"
    Scale="1,1,1"
   ) {
    rotationMode=BMM_Additive
    rotationSpace=BCS_BoneSpace
    translationMode=BMM_Additive
    translationSpace=BCS_BoneSpace
    alphaBoolBlend {blendOption=Linear}
    boneToModify {boneName=dyn_hat}
    rotation {yaw=-3}
    translation=<-1 0.25>
   }
   n2 = "Transform (Modify) Bone - Bone: dyn_hat"(ComponentPose=@n1, Alpha=Get NeckRotRight(), Translation="-0.5,0,0", Rotation="0,0,0", Scale="1,1,1") {
    rotationMode=BMM_Additive
    rotationSpace=BCS_BoneSpace
    translationMode=BMM_Additive
    translationSpace=BCS_BoneSpace
    alphaBoolBlend {blendOption=Linear}
    boneToModify {boneName=dyn_hat}
    translation {x=-0.5}
   }
   n3 = "Constraint - dyn_headphones_MSTR (P)"(ComponentPose=@n2, Alpha=0.8, ConstraintWeights_0=1) {
    alphaBoolBlend {blendOption=Linear}
    boneToModify {boneName=dyn_headphones_MSTR}
    constraintSetup [{targetBone {boneName=spine_05}}]
    constraintWeights [1]
   }
   n4 = "Transform (Modify) Bone - Bone: dyn_headphones_MSTR"(ComponentPose=@n3, Alpha=Get NeckRotDown() -> (NeckRotDown:double), Translation="0.3,0,0", Rotation="0.000,-2,0.000", Scale="1,1,1") {
    rotationMode=BMM_Additive
    rotationSpace=BCS_BoneSpace
    translationMode=BMM_Additive
    translationSpace=BCS_BoneSpace
    alphaBoolBlend {blendOption=Linear}
    boneToModify {boneName=dyn_headphones_MSTR}
    rotation {yaw=-2}
    translation {x=0.3}
   }
   n5 = RigidBody(ComponentPose=@n4) {
    bOverrideWorldGravity=true
    overridePhysicsAsset=&/M_MED_Color_Block_Tinfoil_FaceAcc_CordPh
    simulationSpace=BaseBoneSpace
    alphaBoolBlend {blendOption=Linear}
    baseBoneRef {boneName=root}
    componentLinearAccScale=<0.3 0.3 0.3>
    componentLinearVelScale=<0.2 0.2 0.25>
    overrideWorldGravity {z=-980}
   }
   n6 = "Transform (Modify) Bone - Bone: dyn_cord_MSTR"(ComponentPose=@n5, Alpha=Get RBAN() -> (RBAN:double), Translation="0,0,0", Rotation="0,85,0", Scale="1,1,1") {
    rotationMode=BMM_Replace
    rotationSpace=BCS_BoneSpace
    scaleMode=BMM_Replace
    scaleSpace=BCS_BoneSpace
    translationMode=BMM_Replace
    translationSpace=BCS_BoneSpace
    alphaBoolBlend {blendOption=Linear}
    boneToModify {boneName=dyn_cord_MSTR}
    rotation {yaw=85}
   }
   Output Pose(Result=Component To Local(ComponentPose=@n6)) {layerGroup=None}
   n7 = Layered blend per bone(
    BasePose=Copy Pose From Mesh() {bCopyCurves=true, bUseAttachedParent=true, rootBoneToCopy=root}
    BlendPoses_0=Copy Pose From Mesh(SourceMeshComponent=Get Body Mesh() -> (Body Mesh:Skeletal Mesh Component Object Reference)) {bCopyCurves=true, bUseAttachedParent=true, rootBoneToCopy=root}
   ) {layerSetup [{branchFilters [{boneName=dyn_walkman_MSTR}, {boneName=dyn_cord_MSTR}, {boneName=dyn_walkman_01}]}]}
   Save cached pose 'Start Pose'(Pose=@n7) {cacheName=Start Pose}
  ]}
 EventGraph {nodes [
   Event BlueprintUpdateAnimation()
   > n1 = Sequence()
   n2 = GetOwnerSkeletalMeshComponent()
   n3 = BreakRotator(InRot=BreakTransform(InTransform=GetDeltaTransformFromRefPose(self=@n2, BoneName=neck_02, BaseName=spine_05)).Rotation)
   > n4 = Set NeckRotDown(NeckRotDown:double=MapRangeClamped(Value=@n3.Yaw, InRangeA=0, InRangeB=10)) -> (Output_Get:double)
   > n5 = Branch(Condition="float >= float"(A=@n4, B=0) {is=title K2Node_CallFunction})
   > n6 = Set NeckRotRight(NeckRotRight:double=MapRangeClamped(Value=@n3, InRangeA=8, InRangeB=15)) -> (Output_Get:double)
   n7 = BreakTransform(InTransform=GetDeltaTransformFromRefPose(self=@n2, BoneName=head, BaseName=spine_05))
   Set HeadBackRot(execute=[@n6.then, @n8.then], HeadBackRot:double=MapRangeClamped(Value=BreakRotator(InRot=@n7.Rotation).Yaw, InRangeA=-10, InRangeB=-25)) -> (Output_Get:double)
   n8 = Set NeckRotRight(execute=@n5.else, NeckRotRight=0)
   n9 = Branch(execute=@n1.then_1, Condition=Get bRigidBodyEnabled())
   > Set RBAN(RBAN:double=0) -> (Output_Get:double)
   Set RBAN(execute=@n9.else, RBAN=1)
   Event OnCharacterCustomizationLoaded()
   > n10 = Cast To FortPlayerPawnAthena(Object=TryGetPawnOwner())
   > Set Body Mesh(Body Mesh:Skeletal Mesh Component Object Reference=GetSkeletalMeshForPartType(self=@n10)) -> (Output_Get:Skeletal Mesh Component Object Reference)
  ]}
]
variables [
 Body Mesh {type=SkeletalMeshComponent}
 HeadBackRot {type=number}
 NeckRotDown {type=number}
 NeckRotRight {type=number}
 RBAN {type=number}
]
